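# Light function material in the Nexus environment set that modulates a light with animated, swirling turbulence.
# World-XY cloud noise distorts two tilings of a turbulence texture, which counter-rotate, multiply and pass through a square root to set brightness.
# Shading-path and feature-level switches substitute constant white on mobile and low-end levels.
# A MakeFloat3 branch with time-offset samples for separate channels appears to be left unused.
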Material /Figment_S08/Athena/Environments/Nexus/Materials/FXshader/M_LightFunction
graph {
 properties {materialDomain=MD_LightFunction, pixelDepthOffsetMode=PDOM_Legacy, shadingModel=MSM_Unlit, usedShadingModels=Unlit}
 nodes [
  n1 = WorldPosition()
  n2 = TextureSample(UVs=Divide(A=ComponentMask(None=@n1.XYZ) {g=true, r=true}) {constB=4096}) {samplerType=SAMPLERTYPE_LinearColor, texture=/Game/Missions/Primary/DeliverTheBomb/Textures/T_CloudNoise}
  n3 = Multiply(A=Subtract(A=AppendVector(A=@n2.R, B=@n2.R)) {constB=0.5}) {constB=0.2}
  n4 = Add(A=TextureCoordinate() {uTiling=4, vTiling=4}, B=@n3)
  n5 = Add(A=TextureCoordinate() {uTiling=8, vTiling=8}, B=@n3)
  n6 = Multiply(
   A=TextureSample(UVs=Rotator(Coordinate=@n4) {centerX=2, centerY=2}) {texture=/Game/Effects/Fort_Effects/Textures/Plasma/T_Turbulence_Seamless}.R
   B=TextureSample(UVs=Rotator(Coordinate=@n5) {centerX=4, centerY=4, speed=-0.1}) {texture=&T_Turbulence_Seamless}.R
  )
  n7 = LinearInterpolate(Alpha=SquareRoot(None=Multiply(A=@n6) {constB=2})) {constA=0.01}
  n8 = Constant() {r=1}
  n9 = ShadingPathSwitch(Default=@n7, Mobile=@n8)
  FeatureLevelSwitch(Default=@n9, ES2=@n8, ES3_1=@n8, SM4_REMOVED=@n9)
  n10 = Time()
  n11 = Add(A=@n10) {constB=0.1}
  n12 = TextureSample(UVs=Rotator(Coordinate=TextureCoordinate() {uTiling=4, vTiling=4}, Time=@n11) {centerX=2, centerY=2}) {texture=&T_Turbulence_Seamless}
  n13 = TextureSample(UVs=Rotator(Coordinate=TextureCoordinate() {uTiling=8, vTiling=8}, Time=@n11) {centerX=4, centerY=4, speed=-0.1}) {texture=&T_Turbulence_Seamless}
  n14 = Add(A=@n10) {constB=0.2}
  n15 = TextureSample(UVs=Rotator(Coordinate=TextureCoordinate() {uTiling=4, vTiling=4}, Time=@n14) {centerX=2, centerY=2}) {texture=&T_Turbulence_Seamless}
  n16 = TextureSample(UVs=Rotator(Coordinate=TextureCoordinate() {uTiling=8, vTiling=8}, Time=@n14) {centerX=4, centerY=4, speed=-0.1}) {texture=&T_Turbulence_Seamless}
  MaterialFunctionCall(X=@n7, Y=LinearInterpolate(Alpha=Add(A=@n12.R, B=@n13.R)) {constA=0.25}, Z=LinearInterpolate(Alpha=Add(A=@n15.R, B=@n16.R)) {constA=0.25}) {materialFunction=&MakeFloat3}
  WorldPosition()
 ]
}
