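# Post-process material for the Nexus environment that adds a twitching RGB split and a blue tint to scene color after depth of field.
# It samples the scene three times with center-scaled UVs, offsetting red, green and blue differently by MPC_Nexus PP_Twitch times temporal dither noise.
# The recombined channels receive an additive light-blue color scaled by PP_Intensity from the same collection.
# The local Twitch scalar parameter is unconnected.
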Material /Figment_S08/Athena/Environments/Nexus/Materials/FXshader/M_PP
graph {
 properties {
  blendMode=BLEND_Translucent
  blendableLocation=BL_SceneColorAfterDOF
  materialDomain=MD_PostProcess
  pixelDepthOffsetMode=PDOM_Legacy
  shadingModel=MSM_Unlit
  usedShadingModels=Unlit
 }
 nodes [
  n1 = TextureCoordinate()
  n2 = CollectionParameter() {collection=/Game/Athena/Environments/Nexus/Materials/FXshader/MPC_Nexus, parameterName=PP_Twitch}
  n3 = MaterialFunctionCall() {materialFunction=&DitherTemporalAA}
  n4 = Multiply(A=@n2, B=@n3.Result)
  n5 = Add(A=@n4)
  n6 = ComponentMask(None=SceneTexture(UVs=MaterialFunctionCall(UVs=@n1, Texture Scale=@n5) {materialFunction=&ScaleUVsByCenter}.UVs) {sceneTextureId=PPI_PostProcessInput0}.Color) {r=true}
  n7 = Subtract(A=@n3.Result) {constB=0.5}
  n8 = Add(A=Multiply(A=@n2, B=Multiply(A=@n7) {constB=0.25}))
  n9 = ComponentMask(None=SceneTexture(UVs=MaterialFunctionCall(UVs=@n1, Texture Scale=@n8) {materialFunction=&ScaleUVsByCenter}.UVs) {sceneTextureId=PPI_PostProcessInput0}.Color) {g=true}
  n10 = Subtract(B=@n4)
  n11 = ComponentMask(None=SceneTexture(UVs=MaterialFunctionCall(UVs=@n1, Texture Scale=@n10) {materialFunction=&ScaleUVsByCenter}.UVs) {sceneTextureId=PPI_PostProcessInput0}.Color) {b=true}
  n12 = SceneTexture() {sceneTextureId=PPI_PostProcessInput0}
  n13 = CollectionParameter() {collection=&MPC_Nexus, parameterName=PP_Intensity}
  n14 = Constant3Vector() {constant=<rgba 0.1562 0.4331 1 1>}
  Add(
   A=AppendVector(
    A=MaterialFunctionCall(X=@n6, Y=@n9, Z=@n11) {materialFunction=&MakeFloat3}.Result
    B=ComponentMask(None=@n12.Color) {a=true}
   )
   B=AppendVector(A=Multiply(A=@n13, B=@n14), B=Constant())
  )
  ScalarParameter() {defaultValue=0.1, parameterName=Twitch}
 ]
}
